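# Skeletal-mesh character master material for RedOasis sports skins, recoloring masked regions with a custom color and counter-driven numbers.
# Between character PreFX and PostFX functions, it applies color and number metallic/roughness/specular, then switchable cloth fuzz, silk, clear coat, thin film and metals LUT.
# A skin FX texture supplies channel masks; UV sphere masks give badges separate smoothstep remapping.
# Optional local-space Fresnel attenuates tattoo color, and a debug switch shows badge masks.
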Material /BRCosmetics/Packages/Fortress_SharedMaterials/Base_Material/M_FN_Character_SPORTS_RedOasis
graph {
 properties {
  bUseMaterialAttributes=true
  bUsedWithMeshDeformer=true
  bUsedWithMorphTargets=true
  bUsedWithSkeletalMesh=true
  pixelDepthOffsetMode=PDOM_Legacy
  shadingModel=MSM_FromMaterialExpression
  subsurfaceProfile=&SS_HeroSkin_02
  usedShadingModels=Default Lit | Clear Coat | Subsurface Profile | Cloth
 }
 nodes [
  n1 = TextureObjectParameter() {group=Skin FX, parameterName=SkinFX_Mask, samplerType=SAMPLERTYPE_Masks, texture=&T_BlackMask}
  n2 = TextureSample(Tex=@n1) {
   samplerType=SAMPLERTYPE_Masks
   texture=/Game/Characters/Player/Male/Medium/Bodies/M_MED_Soccer/Textures/Layout_01/T_M_MALE_Soccer_Demo_M_Med_Soccer_Body_TC
  }
  ChannelMaskParameter(None=@n2.RGBA) {group=Skin FX, parameterName=UV Rotation Channel}
  ChannelMaskParameter(None=@n2.RGBA) {group=Skin FX, parameterName=Custom Numbers Channel, defaultValue=<rgba 0 1 0 0>}
  ChannelMaskParameter(None=@n2.RGBA) {group=Skin FX, parameterName=Dirt}
  n3 = MaterialFunctionCall() {materialFunction=&BaseFN_CharacterFunction_PreFX}
  MaterialFunctionCall(AO=@n3.AO, Distance Blend=@n3.AO, Fresnel=@n3.AO, Gmap=@n3.AO, Pre Skinned Local Position=@n3.AO, Material=NamedRerouteUsage()) {materialFunction=&BaseFN_CharacterFunction_PostFX}
  NamedRerouteDeclaration(None=ChannelMaskParameter(None=@n2.RGBA) {group=Skin FX, maskChannel=Green, parameterName=Silk Channel, defaultValue=<rgba 0 1 0 0>}.R) {name=Silk Channel, nodeColor=<rgba 0.5703 1 0 1>}
  n4 = Add(A=NamedRerouteUsage(), B=NamedRerouteUsage())
  n5 = NamedRerouteUsage()
  n6 = NamedRerouteUsage()
  n7 = GetMaterialAttributes(None=@n6) {attributeGetTypes [
    #1
    #2
    #3
    #4
   ]}
  n8 = VectorParameter() {group=RedOasis, parameterName=CustomColor, defaultValue=<rgba 1 0 0 1>}
  n9 = NamedRerouteUsage()
  n10 = Multiply(A=MaterialFunctionCall() {materialFunction=/Game/Gadgets/Assets/VinderTech_Umbrella/Umbrella_BuildABrella/BAB_Counter/Materials/MF_counter}.Result) {constB=1.3}
  n11 = StaticSwitchParameter(True=OneMinus(None=Saturate(None=@n10)), False=Constant() {r=1}) {defaultValue=true, group=Skin FX, parameterName=UseCounter}
  n12 = Multiply(A=@n9, B=@n11)
  n13 = VectorParameter() {group=RedOasis, parameterName=Fresnel Sphere Mask Location, sortPriority=64, defaultValue=<rgba 0 0 145 1>}
  n14 = ScalarParameter() {defaultValue=0.25, group=RedOasis, parameterName=Fresnel Exp, sortPriority=64}
  n15 = ScalarParameter() {defaultValue=0.2, group=RedOasis, parameterName=Fresnel Base Fraction, sortPriority=64}
  n16 = Multiply(
   A=SphereMask(A=NamedRerouteUsage(), B=@n13.RGB, Radius=ScalarParameter() {defaultValue=60, group=RedOasis, parameterName=Fresnel Radius, sortPriority=64}) {hardnessPercent=0}
   B=Multiply(
    A=Fresnel(ExponentIn=@n14, BaseReflectFractionIn=@n15) {baseReflectFraction=0.1, exponent=1}
    B=ScalarParameter() {defaultValue=1, group=RedOasis, parameterName=Fresnel Intensity, sortPriority=64}
   )
  )
  n17 = Multiply(A=@n8.RGB, B=Saturate(None=Subtract(A=@n12, B=@n16)))
  n18 = NamedRerouteUsage()
  n19 = SetMaterialAttributes!(
   MaterialAttributes=@n7.MaterialAttributes
   Base Color=LinearInterpolate(
    A=@n7.Base Color
    B=StaticSwitchParameter(True=@n17, False=@n8.RGB) {group=RedOasis, parameterName=Use Tattoo Fresnel, sortPriority=64}
    Alpha=@n12
   ) {desc=BaseColor}
   Metallic=LinearInterpolate(A=@n7.Metallic, B=NamedRerouteUsage(), Alpha=@n18) {desc=Metallic}
   Specular=@n7.MaterialAttributes
   Roughness=LinearInterpolate(A=@n7.Roughness, B=NamedRerouteUsage(), Alpha=@n18) {desc=Roughness}
  ) {attributeSetTypes [
    #1
    #2
    #3
    #4
   ]}
  n20 = StaticSwitchParameter(
   True=MaterialFunctionCall!(
    Alpha=NamedRerouteUsage()
    Gmap=NamedRerouteUsage()
    Material=StaticSwitchParameter(True=@n19, False=@n6) {defaultValue=true, group=Skin FX, parameterName=Use Custom Colors}
   ) {materialFunction=/Game/Weapons/WeaponSkins/MaterialFunction/ClothFuzz}.Result
   False=@n19
  ) {group=Skin FX, parameterName=Use Cloth fuzz}
  n21 = SetMaterialAttributes!(
   MaterialAttributes=StaticSwitchParameter(
    True=MaterialFunctionCall!(Alpha=@n5, Material=@n20) {materialFunction=/Game/Weapons/WeaponSkins/MaterialFunction/MF_Silk}.Result
    False=@n20
   ) {group=Skin FX, parameterName=Use Silk}
   Fuzz Color or Subsurface Color=Multiply(A=@n5, B=Constant3Vector())
  ) {desc=Turn off subsurface color, attributeSetTypes [#5]}
  n22 = StaticSwitchParameter(
   True=MaterialFunctionCall!(Alpha=NamedRerouteUsage(), In=@n21) {materialFunction=/Game/Weapons/WeaponSkins/MaterialFunction/Clearcoat}.Result
   False=@n21
  ) {group=Skin FX, parameterName=Use Clear Coat}
  n23 = StaticSwitchParameter(
   True=MaterialFunctionCall!(Alpha=NamedRerouteUsage(), Material Attributes=@n22) {materialFunction=/Game/Weapons/WeaponSkins/MaterialFunction/ThinFilmFunction}.Result
   False=@n22
  ) {group=Skin FX, parameterName=Use Thin Film}
  n24 = StaticSwitchParameter(
   True=MaterialFunctionCall(Mask=NamedRerouteUsage(), Materials=@n23) {materialFunction=&MF_MetalsLUT}.Result
   False=@n23
  ) {group=Skin FX, parameterName=UseMetalLUT}
  n25 = GetMaterialAttributes(None=@n24) {attributeGetTypes [#3]}
  n26 = LinearInterpolate(A=@n25.Specular, B=NamedRerouteUsage(), Alpha=NamedRerouteUsage())
  NamedRerouteDeclaration(None=StaticSwitchParameter(
    True=SetMaterialAttributes!(MaterialAttributes, Emissive Color=@n4) {attributeSetTypes [#6]}
    False=SetMaterialAttributes!(MaterialAttributes=@n25.MaterialAttributes, Specular=@n26) {attributeSetTypes [#3]}
   ) {group=RedOasis, parameterName=*Debug Badge Mask}) {name=MA OUT, nodeColor=<rgba 1 0.4688 0 1>}
  n27 = VectorParameter() {
   group=RedOasis
   parameterName=CustomNumber MetallicSpecularRoughness
   channelNames {a="", b=Roughness, g=Specular, r=Metallic}
   defaultValue=<rgba 0.5 0.05 1 1>
  }
  NamedRerouteDeclaration(None=@n27.Roughness) {name=CustomNumber Roughness, nodeColor=<rgba 1 0 0.7031 1>}
  NamedRerouteDeclaration(None=Multiply(A=@n9, B=OneMinus(None=@n11))) {name=CustomNumber Mask, nodeColor=<rgba 0.1953 0 1 1>}
  NamedRerouteDeclaration(None=@n2.RGBA) {name=SkinFX_RGBA, nodeColor=<rgba 0 1 0.9766 1>}
  NamedRerouteDeclaration(None=ChannelMaskParameter(None=@n2.RGBA) {group=RedOasis, maskChannel=Green, parameterName=ThinFilm Channel, defaultValue=<rgba 0 1 0 0>}.R) {name=ThinFilm Channel, nodeColor=<rgba 1 0.75 0 1>}
  NamedRerouteDeclaration(None=@n3.Pre skinned Local Position) {name=PreSkin_LocPos, nodeColor=<rgba 1 0.2109 0 1>}
  NamedRerouteDeclaration(None=@n2.RGB) {name=SkinFX, nodeColor=<rgba 0 0.3437 1 1>}
  NamedRerouteDeclaration(None=ChannelMaskParameter(None=@n2.RGBA) {group=Skin FX, maskChannel=Green, parameterName=Cloth Channel, defaultValue=<rgba 0 1 0 0>}.R) {name=Cloth Channel, nodeColor=<rgba 0 1 0.2969 1>}
  NamedRerouteDeclaration(None=@n3.Gmap) {name=Gmap, nodeColor=<rgba 0.2891 0 1 1>}
  NamedRerouteDeclaration(None=ChannelMaskParameter(None=@n2.RGBA) {group=Skin FX, maskChannel=Green, parameterName=Clearcoat Channel, defaultValue=<rgba 0 1 0 0>}.R) {name=ClearCoat Channel, nodeColor=<rgba 1 0 0.75 1>}
  NamedRerouteDeclaration(None=@n3.Result) {name=MA IN, nodeColor=<rgba 0 1 0.9063 1>}
  n28 = ChannelMaskParameter(None=@n2.RGBA) {group=RedOasis, maskChannel=Green, parameterName=CustomColor Channel, defaultValue=<rgba 0 1 0 0>}
  n29 = SmoothStep(
   Min=ScalarParameter() {group=RedOasis, parameterName=Mask SmoothStep Min}
   Max=ScalarParameter() {defaultValue=1, group=RedOasis, parameterName=Mask SmoothStep Max}
   Value=@n28.R
  ) {constMax=0.9, constMin=0.6}
  n30 = TextureCoordinate()
  n31 = VectorParameter() {group=RedOasis, parameterName=Badge Position, channelNames {a="", b="", g=Y, r=X}, defaultValue=<rgba 0.56 0.74 0 1>}
  n32 = ComponentMask(None=@n31) {g=true, r=true}
  n33 = ScalarParameter() {defaultValue=0.025, group=RedOasis, parameterName=Badge Radius}
  n34 = ScalarParameter() {group=RedOasis, parameterName=Badge Mask Hardness}
  n35 = LinearInterpolate(
   A=@n29
   B=SmoothStep(
    Min=ScalarParameter() {group=RedOasis, parameterName=Badge SmoothStep Min}
    Max=ScalarParameter() {defaultValue=1, group=RedOasis, parameterName=Badge SmoothStep Max}
    Value=@n28.R
   ) {constMax=0.9, constMin=0.6}
   Alpha=NamedRerouteDeclaration(None=SphereMask(A=@n30, B=@n32, Radius=@n33, Hardness=@n34) {attenuationRadius=0.1, hardnessPercent=0}) {name=Badge Mask, nodeColor=<rgba 1 0.2109 0 1>}
  )
  n36 = SmoothStep(
   Min=ScalarParameter() {group=RedOasis, parameterName=Badge 2 SmoothStep Min}
   Max=ScalarParameter() {defaultValue=1, group=RedOasis, parameterName=Badge 2 SmoothStep Max}
   Value=@n28.R
  ) {constMax=0.9, constMin=0.6}
  n37 = TextureCoordinate()
  n38 = VectorParameter() {group=RedOasis, parameterName=Badge 2 Position, channelNames {a="", b="", g=Y, r=X}, defaultValue=<rgba 0.82 0.9 0 1>}
  n39 = ComponentMask(None=@n38) {g=true, r=true}
  n40 = ScalarParameter() {defaultValue=0.025, group=RedOasis, parameterName=Badge 2 Radius}
  n41 = ScalarParameter() {group=RedOasis, parameterName=Badge 2 Mask Hardness}
  n42 = VectorParameter() {group=RedOasis, parameterName=Badge 2 Position 2, channelNames {a="", b="", g=Y, r=X}, defaultValue=<rgba 0.82 0.9 0 1>}
  n43 = SphereMask(A=TextureCoordinate(), B=ComponentMask(None=@n42) {g=true, r=true}, Radius=@n40, Hardness=@n41) {attenuationRadius=0.1, hardnessPercent=0}
  n44 = Constant()
  n45 = NamedRerouteDeclaration(None=Max(
    A=SphereMask(A=@n37, B=@n39, Radius=@n40, Hardness=@n41) {attenuationRadius=0.1, hardnessPercent=0}
    B=StaticSwitchParameter(True=@n43, False=@n44) {group=RedOasis, parameterName=Use Badge 2 SphereMask 2, sortPriority=64}
   )) {name=Badge 2 Mask, nodeColor=<rgba 1 0.2109 0 1>}
  NamedRerouteDeclaration(None=StaticSwitchParameter(True=LinearInterpolate(A=@n35, B=@n36, Alpha=@n45), False=@n29) {group=RedOasis, parameterName=Use Smoothstep Badge}) {name=CustomColor Channel, nodeColor=<rgba 0 1 0.5547 1>}
  NamedRerouteDeclaration(None=@n27.Metallic) {name=CustomNumber Metallic, nodeColor=<rgba 0 1 0.7188 1>}
  NamedRerouteDeclaration(None=@n27.Specular) {name=CustomNumber Specular, nodeColor=<rgba 1 0.8438 0 1>}
  SetMaterialAttributes!(MaterialAttributes=@n24, Base Color=NamedRerouteUsage()) {attributeSetTypes [#1]}
 ]
}
