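# Material function for paper-doll customization that recolors and re-finishes input material attributes with masked layers and a tattoo.
# A Gmap scalar drives five gradient lookups; layers 2, 3 and 5 can switch to tiled camo.
# Mask RGBA selects color and roughness (alpha optional), and its R, B, A maximum, omitting G, blends onto the input.
# The tattoo overlays base color last; metallic uses PD_Metalness, leaving the per-layer metalness chain unconnected.
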MaterialFunction /BRCosmetics/Weapons/WeaponSkins/MaterialFunction/MF_PaperDoll_Customization
graph {nodes [
  n1 = FunctionInput() {inputName=Material, inputType=FunctionInput_MaterialAttributes}
  n2 = GetMaterialAttributes(None=@n1) {attributeGetTypes [#1, #2, #3]}
  n3 = Saturate(None=FunctionInput(Preview=TextureSample() {texture=/Game/Global/Textures/Default/Blanks/T_WhiteUserInterface}.R) {bUsePreviewValueAsDefault=true, inputName=Gmap, inputType=FunctionInput_Scalar})
  n4 = TextureSampleParameter2D(UVs=@n3) {
   group=Gmap Layers
   parameterName=Layer1_Gradient
   samplerSource=SSM_Clamp_WorldGroupSettings
   sortPriority=1
   texture=/Game/Characters/Player/Male/Medium/Bodies/M_MED_Dragon_Ninja/Textures/V1/T_DragonNinja_RedArmor
  }
  n5 = AppendVector(A=@n3, B=Constant())
  n6 = AppendVector(
   A=ScalarParameter() {defaultValue=1, group=PaperDoll_Pattern, parameterName=PD_Pattern_U_Tiling}
   B=ScalarParameter() {defaultValue=1, group=PaperDoll_Pattern, parameterName=PD_Pattern_V_Tiling}
  )
  n7 = Multiply(A=TextureCoordinate(), B=@n6)
  n8 = ScalarParameter() {group=PaperDoll_Pattern, parameterName=PD_Use_Layer_02_Camo}
  n9 = TextureSampleParameter2D(UVs=LinearInterpolate(A=@n5, B=@n7, Alpha=@n8)) {
   group=Gmap Layers
   parameterName=Layer2_Gradient
   sortPriority=2
   texture=&T_DragonNinja_RedArmor
  }
  n10 = LinearInterpolate(
   A=ScalarParameter() {desc=Multiply Pattern by GMap, group=PaperDoll_Pattern, parameterName=GMap Mult Min}
   B=ScalarParameter() {defaultValue=1, desc=Multiply Pattern by GMap, group=PaperDoll_Pattern, parameterName=GMap Mult Max}
   Alpha=@n3
  )
  n11 = LinearInterpolate(A=@n9.RGB, B=Saturate(None=Multiply(A=@n9.RGB, B=@n10)), Alpha=@n8)
  n12 = TextureSampleParameter2D(UVs=TextureCoordinate()) {
   group=Gmap Layers
   parameterName=Layer Mask
   samplerSource=SSM_Wrap_WorldGroupSettings
   samplerType=SAMPLERTYPE_Masks
   sortPriority=-2
   texture=&T_BlackMask
  }
  n13 = TextureCoordinate() {coordinateIndex=2}
  n14 = ScalarParameter() {group=PaperDoll_Pattern, parameterName=PD_Use_Layer_03_Camo}
  n15 = TextureSampleParameter2D(UVs=LinearInterpolate(A=@n5, B=Multiply(A=@n13, B=@n6), Alpha=@n14)) {
   group=Gmap Layers
   parameterName=Layer3_Gradient
   sortPriority=3
   texture=&T_DragonNinja_RedArmor
  }
  n16 = Multiply(A=@n15.RGB, B=ScalarParameter() {defaultValue=0.5, group=Gmap Layers, parameterName=Layer3_SleeveDarken})
  n17 = TextureSampleParameter2D(UVs=@n3) {
   group=Gmap Layers
   parameterName=Layer4_Gradient
   samplerSource=SSM_Clamp_WorldGroupSettings
   sortPriority=4
   texture=&T_DragonNinja_RedArmor
  }
  n18 = LinearInterpolate(
   A=LinearInterpolate(A=LinearInterpolate(A=@n4.RGB, B=@n11, Alpha=@n12.R), B=LinearInterpolate(A=@n15.RGB, B=@n16, Alpha=@n14), Alpha=@n12.G)
   B=@n17.RGB
   Alpha=@n12.B
  )
  n19 = ScalarParameter() {group=PaperDoll_Pattern, parameterName=PD_Use_Layer_05_Camo}
  n20 = TextureSampleParameter2D(UVs=LinearInterpolate(A=@n5, B=@n7, Alpha=@n19)) {
   group=Gmap Layers
   parameterName=Layer5_Gradient
   sortPriority=5
   texture=&T_DragonNinja_RedArmor
  }
  n21 = Saturate(None=Multiply(A=@n20.RGB, B=@n10))
  n22 = StaticBoolParameter() {group=Gmap Layers, parameterName=use Alpha Channel as mask, sortPriority=-1}
  n23 = StaticSwitch(True=LinearInterpolate(A=@n18, B=LinearInterpolate(A=@n20.RGB, B=@n21, Alpha=@n19), Alpha=@n12.A), False=@n18, Value=@n22)
  n24 = Multiply(A=@n2.Roughness, B=ScalarParameter() {defaultValue=1, group=Gmap Layers, parameterName=Layer1_Roughness, sortPriority=1})
  n25 = Multiply(A=@n2.Roughness, B=ScalarParameter() {defaultValue=1, group=Gmap Layers, parameterName=Layer2_Roughness, sortPriority=2})
  n26 = ScalarParameter() {defaultValue=1, group=Gmap Layers, parameterName=Layer3_Roughness, sortPriority=3}
  n27 = LinearInterpolate(
   A=LinearInterpolate(A=LinearInterpolate(A=@n24, B=@n25, Alpha=@n12.R), B=Multiply(A=@n2.Roughness, B=@n26), Alpha=@n12.G)
   B=Multiply(A=@n2.Roughness, B=ScalarParameter() {defaultValue=1, group=Gmap Layers, parameterName=Layer4_Roughness, sortPriority=4})
   Alpha=@n12.B
  )
  n28 = ScalarParameter() {defaultValue=1, group=Gmap Layers, parameterName=Layer5_Roughness, sortPriority=5}
  n29 = StaticSwitch(True=LinearInterpolate(A=@n27, B=Multiply(A=@n2.Roughness, B=@n28), Alpha=@n12.A), False=@n27, Value=@n22)
  n30 = ScalarParameter() {parameterName=PD_Metalness}
  n31 = Max(A=Max(A=@n12.R, B=@n12.R), B=@n12.B)
  n32 = GetMaterialAttributes(None=BlendMaterialAttributes(
    A=@n1
    B=SetMaterialAttributes!(MaterialAttributes=@n2.MaterialAttributes, BaseColor=@n23, Roughness=@n29, Metallic=@n30) {attributeSetTypes [#1, #2, #3]}
    Alpha=Max(A=@n31, B=@n12.A)
   )) {attributeGetTypes [#1]}
  n33 = TextureSampleParameter2D(UVs=LinearInterpolate(
    A=TextureCoordinate()
    B=TextureCoordinate() {coordinateIndex=1}
    Alpha=ScalarParameter() {group=PaperDoll_Tattoo, parameterName="PD_UseSecondUVChannel(Tattoo)"}
   )) {
   group=PaperDoll_Tattoo
   parameterName=PO_Tattoo_Layer_01
   texture=/Game/Characters/Player/Female/Medium/Heads/F_Med_Head_01/Skins/Tactical_Desert/Textures/T_BlankAlpha
  }
  n34 = LinearInterpolate(A=@n32.BaseColor, B=@n33.RGB, Alpha=@n33.RGB)
  n35 = ScalarParameter() {defaultValue=1, group=PaperDoll_Tattoo, parameterName=PD_TattooOpacity}
  FunctionOutput(None=SetMaterialAttributes!(MaterialAttributes=@n32.MaterialAttributes, BaseColor=LinearInterpolate(A=@n32.BaseColor, B=@n34, Alpha=@n35)) {attributeSetTypes [#1]})
  n36 = Multiply(A=@n2.Metallic, B=ScalarParameter() {group=Gmap Layers, parameterName=Layer1_Metalness, sortPriority=1})
  n37 = Multiply(A=@n2.Metallic, B=ScalarParameter() {group=Gmap Layers, parameterName=Layer2_Metalness, sortPriority=2})
  n38 = ScalarParameter() {group=Gmap Layers, parameterName=Layer3_Metalness, sortPriority=3}
  n39 = LinearInterpolate(
   A=LinearInterpolate(A=LinearInterpolate(A=@n36, B=@n37, Alpha=@n12.R), B=Multiply(A=@n2.Metallic, B=@n38), Alpha=@n12.G)
   B=Multiply(A=@n2.Metallic, B=ScalarParameter() {group=Gmap Layers, parameterName=Layer4_Metalness, sortPriority=4})
   Alpha=@n12.B
  )
  n40 = Multiply(A=@n2.Metallic, B=ScalarParameter() {group=Gmap Layers, parameterName=Layer5_Metalness, sortPriority=5})
  n41 = ScalarParameter() {group=Gmap Layers, parameterName=Layer1_Metalness, sortPriority=1}
  n42 = ScalarParameter() {group=Gmap Layers, parameterName=Layer2_Metalness, sortPriority=2}
  n43 = LinearInterpolate(
   A=LinearInterpolate(
    A=LinearInterpolate(A=@n41, B=@n42, Alpha=@n12.R)
    B=ScalarParameter() {group=Gmap Layers, parameterName=Layer3_Metalness, sortPriority=3}
    Alpha=@n12.G
   )
   B=ScalarParameter() {group=Gmap Layers, parameterName=Layer4_Metalness, sortPriority=4}
   Alpha=@n12.B
  )
  n44 = ScalarParameter() {group=Gmap Layers, parameterName=Layer5_Metalness, sortPriority=5}
  LinearInterpolate(
   A=StaticSwitch(True=LinearInterpolate(A=@n39, B=@n40, Alpha=@n12.A), False=@n39, Value=@n22)
   B=StaticSwitch(True=LinearInterpolate(A=@n43, B=@n44, Alpha=@n12.A), False=@n43, Value=@n22)
   Alpha=ScalarParameter() {defaultValue=1, group=Gmap Layers, parameterName=Overwrite Metallic from _S Mask}
  )
  Saturate(None=Multiply(A=@n15.RGB, B=@n10))
 ]}
