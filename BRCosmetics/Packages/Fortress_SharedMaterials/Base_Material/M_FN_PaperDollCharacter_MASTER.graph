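# Master character material for Fortnite paper-doll skeletal meshes, using material attributes and the hero skin subsurface profile.
# A base character function supplies attributes and a mask split into AO, GMap and skin, feeding GMap color, brightness balance and skin functions.
# A static switch adds optional paper-doll customization before rim lighting, tech-art overrides, de-rez transition and opacity.
# The emissive function and SkinFXMask branches feed nothing downstream.
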Material /BRCosmetics/Packages/Fortress_SharedMaterials/Base_Material/M_FN_PaperDollCharacter_MASTER
graph {
 properties {
  bAutomaticallySetUsageInEditor=false
  bUseMaterialAttributes=true
  bUsedWithMeshDeformer=true
  bUsedWithMorphTargets=true
  bUsedWithSkeletalMesh=true
  pixelDepthOffsetMode=PDOM_Legacy
  shadingModel=MSM_FromMaterialExpression
  subsurfaceProfile=&SS_HeroSkin_02
  usedShadingModels=Default Lit | Subsurface Profile
 }
 nodes [
  Constant4Vector() {constant=<rgba 1 1 1 1>}
  Constant() {r=1}
  n1 = MaterialFunctionCall!(TexCoords) {materialFunction=/Game/Packages/Fortress_SharedMaterials/Base_Material_Functions/BaseCharacterShaderFunctions/BaseFN_CharacterFunction}
  n2 = MaterialFunctionCall(Float3=@n1.M) {desc="R = AO\r\nG = GMap\r\nB = Skin", materialFunction=&BreakOutFloat3Components}
  n3 = MaterialFunctionCall!(Material=@n1.Result, Diffuse=@n1.Result, GMap=@n2.G) {materialFunction=/Game/Packages/Fortress_SharedMaterials/Base_Material_Functions/BaseCharacterShaderFunctions/BaseFN_CharacterGMapColor}
  n4 = MaterialFunctionCall!(Material=@n3.Result) {materialFunction=/Game/Packages/Fortress_SharedMaterials/Base_Material_Functions/CharacterBrightnessandScale_Balance}
  n5 = PixelNormalWS()
  n6 = CameraVectorWS()
  MaterialFunctionCall!(
   Material=@n4.Result
   Diffuse=@n3.DIffuse
   BalanceDistance=@n4.Result
   Normals=@n1.Normal Map
   AO=@n2.R
   Fresnel=Saturate(None=DotProduct(A=@n5, B=@n6))
  ) {materialFunction=/Game/Packages/Fortress_SharedMaterials/Base_Material_Functions/BaseCharacterShaderFunctions/BaseFN_CharacterEmissive}
  n7 = TextureSampleParameter2D() {
   parameterName=SkinFXMask
   texture=/BRCosmetics/Characters/Player/Female/Medium/Bodies/F_MED_Desert_Ops_Camo/Textures/Base/T_F_MED_Desert_Ops_Camo_Base_Body_FX
  }
  Max(A=Max(A=@n7.R, B=@n7.R), B=@n7.B)
  n8 = MaterialFunctionCall!(Material=@n4.Result, AO=@n2.R, Skin Mask=@n2.B) {materialFunction=/Game/Packages/Fortress_SharedMaterials/Base_Material_Functions/BaseCharacterShaderFunctions/BaseFN_CharacterSkin}
  n9 = MaterialFunctionCall!(Material=@n8.Result, Gmap=@n2.G) {materialFunction=/BRCosmetics/Weapons/WeaponSkins/MaterialFunction/MF_PaperDoll_Customization}
  n10 = MaterialFunctionCall(In=MaterialFunctionCall!(
    AO=@n2.R
    Distance Blend=@n4.Distance Blend
    Gmap
    Pre Skinned Local Position
    Material=StaticSwitchParameter(True=@n9.Result, False=@n8.Result) {parameterName=UsePaperDollCustomization}
   ) {materialFunction=/BRCosmetics/Packages/Fortress_SharedMaterials/Base_Material_Functions/MF_Rimv2}.Result) {materialFunction=&MF_TechArtOverrides}
  StaticSwitch(
   True=MaterialFunctionCall!(Material=MaterialFunctionCall!(Fresnel, Material=@n10.Result) {materialFunction=/Game/Packages/Fortress_SharedMaterials/Base_Material_Functions/CharacterDeRez_Transition}.Result) {materialFunction=/Game/Packages/Fortress_SharedMaterials/Base_Material_Functions/BaseCharacterShaderFunctions/BaseFN_CharacterOpacity}.Result
   False=SetMaterialAttributes!(MaterialAttributes, Shading Model=ShadingModel() {shadingModel=MSM_SubsurfaceProfile}) {attributeSetTypes [#1]}
  ) {defaultValue=true}
 ]
}
